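# Sound cue for the Chrono Cloak activation beeps, playing one wave with a different mix per listener.
# An IsLocal branch plays it at full volume through first-person attenuation for the local player, otherwise deferring to an IsEnemy branch.
# Enemy or unset plays at reduced volume and non-enemy much quieter, both through standard cloak attenuation with slight random pitch variation.
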SoundCue /CloakGameplay/Audio/SFX/ChronoCloak_Assets/Chrono_Cloak_Activate_Beeps_Cue
root=Branch_0
nodes [
 {id=Attenuation_0, children [WavePlayer_0], properties {attenuationSettings=/CloakGameplay/Audio/SFX/Fort_Cloak_Activate_1P_Att}}
 {id=Attenuation_1, children [Modulator_0], properties {attenuationSettings=/CloakGameplay/Audio/SFX/Fort_Cloak_Activate_Att}}
 {id=Attenuation_2, children [Modulator_1], properties {attenuationSettings=&Fort_Cloak_Activate_Att}}
 {id=Branch_0, children [Mixer_0,Branch_1,Branch_1], properties {boolParameterName=IsLocal}}
 {id=Branch_1, children [Mixer_1,Mixer_2,Mixer_1], properties {boolParameterName=IsEnemy}}
 {id=Mixer_0, children [Attenuation_0], properties {inputVolume [1]}}
 {id=Mixer_1, children [Attenuation_1], properties {inputVolume [0.6]}}
 {id=Mixer_2, children [Attenuation_2], properties {inputVolume [0.1]}}
 {id=Modulator_0, children [WavePlayer_2], properties {pitchMax=1.1, pitchMin=0.9, volumeMax=1, volumeMin=1}}
 {id=Modulator_1, children [WavePlayer_9], properties {pitchMax=1.1, pitchMin=0.9, volumeMax=1, volumeMin=1}}
 {id=WavePlayer_0, properties {soundWaveAssetPtr=&/Chrono_Cloak_Activate_Beeps_01}}
 {id=WavePlayer_2, properties {soundWaveAssetPtr=&Chrono_Cloak_Activate_Beeps_01}}
 {id=WavePlayer_9, properties {soundWaveAssetPtr=&Chrono_Cloak_Activate_Beeps_01}}
]
properties {
 attenuationSettings=&Fort_Cloak_Activate_Att
 maxDistance=12500
 soundClassObject=&Fort_Gadgets
 volumeMultiplier=1
}
