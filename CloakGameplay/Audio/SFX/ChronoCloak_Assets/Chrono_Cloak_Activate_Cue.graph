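# Sound cue for the Chrono Cloak activation sound in CloakGameplay, classed as a gadget sound.
# An IsLocal branch plays a dedicated first-person wave with 1P attenuation, otherwise handing off to an IsEnemy branch.
# Both third-person paths randomly pick among the same three start waves under shared attenuation, with non-enemies mixed much quieter.
# An unset IsLocal is treated as remote and an unset IsEnemy as enemy.
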SoundCue /CloakGameplay/Audio/SFX/ChronoCloak_Assets/Chrono_Cloak_Activate_Cue
root=Branch_0
nodes [
 {id=Attenuation_0, children [WavePlayer_0], properties {attenuationSettings=/CloakGameplay/Audio/SFX/Fort_Cloak_Activate_1P_Att}}
 {id=Attenuation_1, children [Random_0], properties {attenuationSettings=/CloakGameplay/Audio/SFX/Fort_Cloak_Activate_Att}}
 {id=Attenuation_2, children [Random_1], properties {attenuationSettings=&Fort_Cloak_Activate_Att}}
 {id=Branch_0, children [Mixer_0,Branch_1,Branch_1], properties {boolParameterName=IsLocal}}
 {id=Branch_1, children [Mixer_1,Mixer_2,Mixer_1], properties {boolParameterName=IsEnemy}}
 {id=Mixer_0, children [Attenuation_0], properties {inputVolume [1]}}
 {id=Mixer_1, children [Attenuation_1], properties {inputVolume [1]}}
 {id=Mixer_2, children [Attenuation_2], properties {inputVolume [0.2]}}
 {id=Random_0, children [WavePlayer_3,WavePlayer_4,WavePlayer_5], properties {bShouldExcludeFromBranchCulling=true, weights [1,1,1]}}
 {id=Random_1, children [WavePlayer_6,WavePlayer_7,WavePlayer_8], properties {bShouldExcludeFromBranchCulling=true, weights [1,1,1]}}
 {id=WavePlayer_0, properties {soundWaveAssetPtr=&/Chrono_Cloak_Activate_Start_1P}}
 {id=WavePlayer_3, properties {soundWaveAssetPtr=&/Chrono_Cloak_Activate_Start_01}}
 {id=WavePlayer_4, properties {soundWaveAssetPtr=&/Chrono_Cloak_Activate_Start_02}}
 {id=WavePlayer_5, properties {soundWaveAssetPtr=&/Chrono_Cloak_Activate_Start_03}}
 {id=WavePlayer_6, properties {soundWaveAssetPtr=&Chrono_Cloak_Activate_Start_01}}
 {id=WavePlayer_7, properties {soundWaveAssetPtr=&Chrono_Cloak_Activate_Start_02}}
 {id=WavePlayer_8, properties {soundWaveAssetPtr=&Chrono_Cloak_Activate_Start_03}}
]
properties {
 attenuationSettings=&Fort_Cloak_Activate_Att
 maxDistance=12500
 soundClassObject=&Fort_Gadgets
 volumeMultiplier=0.6
}
