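# Sound cue that plays the Chrono Cloak deactivation sound, classed as a gadget sound.
# An IsLocal branch picks the local path, or falls through to an IsEnemy branch when false or unset.
# All paths play the same end wave through the shared cloak-activate attenuation, with single-input mixers setting volume.
# Non-local non-enemies hear it much quieter, while local, enemy and unset cases keep full input volume.
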SoundCue /CloakGameplay/Audio/SFX/ChronoCloak_Assets/Chrono_Cloak_Deactivate_Cue
root=Branch_0
nodes [
 {id=Attenuation_0, children [WavePlayer_0], properties {attenuationSettings=/CloakGameplay/Audio/SFX/Fort_Cloak_Activate_Att}}
 {id=Attenuation_1, children [WavePlayer_4], properties {attenuationSettings=&Fort_Cloak_Activate_Att}}
 {id=Attenuation_2, children [WavePlayer_5], properties {attenuationSettings=&Fort_Cloak_Activate_Att}}
 {id=Branch_0, children [Mixer_0,Branch_1,Branch_1], properties {boolParameterName=IsLocal}}
 {id=Branch_1, children [Mixer_1,Mixer_2,Mixer_1], properties {boolParameterName=IsEnemy}}
 {id=Mixer_0, children [Attenuation_0], properties {inputVolume [1]}}
 {id=Mixer_1, children [Attenuation_1], properties {inputVolume [1]}}
 {id=Mixer_2, children [Attenuation_2], properties {inputVolume [0.2]}}
 {id=WavePlayer_0, properties {soundWaveAssetPtr=&/Chrono_Cloak_End}}
 {id=WavePlayer_4, properties {soundWaveAssetPtr=&Chrono_Cloak_End}}
 {id=WavePlayer_5, properties {soundWaveAssetPtr=&Chrono_Cloak_End}}
]
properties {
 attenuationSettings=&Fort_Cloak_Activate_Att
 maxDistance=6500
 soundClassObject=&Fort_Gadgets
 volumeMultiplier=1.6
}
